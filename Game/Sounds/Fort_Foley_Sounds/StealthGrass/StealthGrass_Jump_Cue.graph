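# Sound cue for jumping in stealth grass, layering two randomized fern jungle grass footstep samples as player foley.
# One of four footsteps plays immediately at reduced volume, mixed with one of four others after a short fixed delay to form a double rustle.
# A modulator slightly varies the pitch of the whole mix.
# Foley attenuation applies, and a per-owner concurrency override caps simultaneous copies, stopping the oldest.
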SoundCue /Game/Sounds/Fort_Foley_Sounds/StealthGrass/StealthGrass_Jump_Cue
root=Modulator_1
nodes [
 {id=Delay_0, children [Random_6], properties {delayMax=0.05, delayMin=0.05}}
 {id=Mixer_0, children [Random_5,Delay_0], properties {inputVolume [0.7,1]}}
 {id=Modulator_1, children [Mixer_0], properties {pitchMax=1.1, pitchMin=0.9}}
 {id=Random_5, children [WavePlayer_15, WavePlayer_17, WavePlayer_18, WavePlayer_19], properties {weights [1,1,1,1]}}
 {id=Random_6, children [WavePlayer_20, WavePlayer_21, WavePlayer_22, WavePlayer_23], properties {weights [1,1,1,1]}}
 {id=WavePlayer_15, properties {soundWaveAssetPtr=/Game/Sounds/Fort_Foley_Sounds/FernJungle/FernJungleGrass_Footstep_01}}
 {id=WavePlayer_17, properties {soundWaveAssetPtr=/Game/Sounds/Fort_Foley_Sounds/FernJungle/FernJungleGrass_Footstep_02}}
 {id=WavePlayer_18, properties {soundWaveAssetPtr=/Game/Sounds/Fort_Foley_Sounds/FernJungle/FernJungleGrass_Footstep_03}}
 {id=WavePlayer_19, properties {soundWaveAssetPtr=/Game/Sounds/Fort_Foley_Sounds/FernJungle/FernJungleGrass_Footstep_04}}
 {id=WavePlayer_20, properties {soundWaveAssetPtr=/Game/Sounds/Fort_Foley_Sounds/FernJungle/FernJungleGrass_Footstep_05}}
 {id=WavePlayer_21, properties {soundWaveAssetPtr=/Game/Sounds/Fort_Foley_Sounds/FernJungle/FernJungleGrass_Footstep_06}}
 {id=WavePlayer_22, properties {soundWaveAssetPtr=/Game/Sounds/Fort_Foley_Sounds/FernJungle/FernJungleGrass_Footstep_07}}
 {id=WavePlayer_23, properties {soundWaveAssetPtr=/Game/Sounds/Fort_Foley_Sounds/FernJungle/FernJungleGrass_Footstep_08}}
]
properties {
 attenuationSettings=&Foley_Attenuation_Basic_Att
 bOverrideConcurrency=true
 maxDistance=1536
 soundClassObject=&Fort_Player_Foley_1P_Athena
 volumeMultiplier=1.3
 concurrencyOverrides {bLimitToOwner=true, maxCount=3, resolutionRule=StopOldest, volumeScale=0.8}
}
